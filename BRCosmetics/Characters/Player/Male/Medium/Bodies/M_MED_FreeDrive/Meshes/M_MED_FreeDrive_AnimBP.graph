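# Animation Blueprint for the FreeDrive medium male player body, adding secondary motion to clothing and accessory bones.
# It copies the parent mesh pose into a Control Rig that receives a front-end idle curve value.
# Sleeve bones constrain to hands, then AnimDynamics simulates chains, collar, belts and a zipper pre-positioned by weighted multi-bone constraint.
# Each update sets simulation-space and alpha variables from front-end state, with DBNO-carried also affecting simulation space.
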AnimBlueprint /BRCosmetics/Characters/Player/Male/Medium/Bodies/M_MED_FreeDrive/Meshes/M_MED_FreeDrive_AnimBP
target_skeleton=&/M_MED_FreeDrive_Skeleton
graphs [
 AnimGraph {nodes [
   n1 = Control Rig!(
    Source=Copy Pose From Mesh() {bUseAttachedParent=true}
    Alpha
    frontendIdle:double=Get Custom Idle is Playing Float() -> (Custom Idle is Playing Float:float)
   ) -> (Pose) {
    alphaBoolBlend {blendOption=Linear}
    controlRigAssetReference {blueprintRigClass=&/M_MED_FreeDrive_CtrlRig.M_MED_FreeDrive_CtrlRig_C}
   }
   n2 = "Constraint - dyn_L_sleeveRotate (P)"(ComponentPose=Local To Component(LocalPose=@n1), Alpha=1, ConstraintWeights_0=1) {
    alphaBoolBlend {blendOption=Linear}
    boneToModify {boneName=dyn_L_sleeveRotate}
    constraintSetup [{targetBone {boneName=hand_l}}]
    constraintWeights [1]
   }
   n3 = "Constraint - dyn_R_sleeveRotate (P)"(ComponentPose=@n2, Alpha=1, ConstraintWeights_0=1) {
    alphaBoolBlend {blendOption=Linear}
    boneToModify {boneName=dyn_R_sleeveRotate}
    constraintSetup [{targetBone {boneName=hand_r}}]
    constraintWeights [1]
   }
   n4 = Anim Dynamics - Bone: dyn_chain_upper!(SimulationSpace:AnimPhysSimSpaceType Enum, ComponentPose=@n3, Alpha=1) -> (Pose) {
    angularDampingOverride=1
    angularSpringConstant=5000
    bAngularSpring=true
    bOverrideAngularDamping=true
    bOverrideLinearDamping=true
    bUseGravityOverride=true
    bUsePlanarLimit=false
    gravityScale=0.35
    linearDampingOverride=1
    simulationSpace=RootRelative
    alphaBoolBlend {blendOption=Linear}
    alphaScaleBiasClamp {scale=0.2}
    boundBone {boneName=dyn_chain_upper}
    componentAppliedLinearAccClamp=<5000 5000 5000>
    componentLinearAccScale=<0.1 0.1 0.1>
    componentLinearVelScale=<0.1 0.1 0.1>
    gravityOverride {z=-2000}
    physicsBodyDefinitions [{
      boundBone {boneName=dyn_chain_upper}
      boxExtents=<800 800 800>
      constraintSetup {angularLimitsMax=<30 0 0>, angularLimitsMin=<-30 0 0>, angularTarget=<1 0 0>, linearAxesMax=<10 0 0>, linearAxesMin=<-20 0 0>}
      localJointOffset=<-2 0 0>
     }]
   }
   n5 = Anim Dynamics - Bone: dyn_chain_upperSecondary!(SimulationSpace:AnimPhysSimSpaceType Enum, ComponentPose=@n4, Alpha=1) -> (Pose) {
    angularDampingOverride=1
    angularSpringConstant=5000
    bAngularSpring=true
    bOverrideAngularDamping=true
    bOverrideLinearDamping=true
    bUseGravityOverride=true
    bUsePlanarLimit=false
    gravityScale=0.35
    linearDampingOverride=1
    simulationSpace=RootRelative
    alphaBoolBlend {blendOption=Linear}
    alphaScaleBiasClamp {scale=0.2}
    boundBone {boneName=dyn_chain_upperSecondary}
    componentAppliedLinearAccClamp=<5000 5000 5000>
    componentLinearAccScale=<0.1 0.1 0.1>
    componentLinearVelScale=<0.1 0.1 0.1>
    gravityOverride {z=-2000}
    physicsBodyDefinitions [{
      boundBone {boneName=dyn_chain_upperSecondary}
      boxExtents=<700 700 700>
      constraintSetup {angularLimitsMax=<30 0 0>, angularLimitsMin=<-30 0 0>, angularTarget=<1 0 0>, linearAxesMax=<25 0 0>, linearAxesMin=<-30 0 0>}
      localJointOffset=<-2 0 0>
     }]
   }
   n6 = Anim Dynamics - Bone: dyn_chain_secondary!(SimulationSpace:AnimPhysSimSpaceType Enum, ComponentPose=@n5, Alpha=1) -> (Pose) {
    angularDampingOverride=1
    angularSpringConstant=5000
    bAngularSpring=true
    bOverrideAngularDamping=true
    bOverrideLinearDamping=true
    bUseGravityOverride=true
    bUsePlanarLimit=false
    gravityScale=0.35
    linearDampingOverride=1
    simulationSpace=RootRelative
    alphaBoolBlend {blendOption=Linear}
    alphaScaleBiasClamp {scale=0.2}
    boundBone {boneName=dyn_chain_secondary}
    componentAppliedLinearAccClamp=<5000 5000 5000>
    componentLinearAccScale=<0.1 0.1 0.1>
    componentLinearVelScale=<0.1 0.1 0.1>
    gravityOverride {z=-2000}
    physicsBodyDefinitions [{
      boundBone {boneName=dyn_chain_secondary}
      boxExtents=<700 700 700>
      constraintSetup {angularLimitsMax=<30 0 0>, angularLimitsMin=<-30 0 0>, angularTarget=<1 0 0>, linearAxesMax=<20 0 0>, linearAxesMin=<-20 0 0>}
      localJointOffset=<-2 0 0>
     }]
   }
   n7 = Anim Dynamics - Bone: dyn_chain_lower!(SimulationSpace:AnimPhysSimSpaceType Enum, ComponentPose=@n6, Alpha=1) -> (Pose) {
    angularDampingOverride=1
    angularSpringConstant=5000
    bAngularSpring=true
    bOverrideAngularDamping=true
    bOverrideLinearDamping=true
    bUseGravityOverride=true
    bUsePlanarLimit=false
    gravityScale=0.35
    linearDampingOverride=1
    simulationSpace=RootRelative
    alphaBoolBlend {blendOption=Linear}
    alphaScaleBiasClamp {scale=0.2}
    boundBone {boneName=dyn_chain_lower}
    componentAppliedLinearAccClamp=<5000 5000 5000>
    componentLinearAccScale=<0.1 0.1 0.1>
    componentLinearVelScale=<0.1 0.1 0.1>
    gravityOverride {z=-2000}
    physicsBodyDefinitions [{
      boundBone {boneName=dyn_chain_lower}
      boxExtents=<700 700 700>
      constraintSetup {angularLimitsMax=<30 0 0>, angularLimitsMin=<-30 0 0>, angularTarget=<1 0 0>, linearAxesMax=<15 0 0>, linearAxesMin=<-20 0 0>}
      localJointOffset=<-2 0 0>
     }]
   }
   n8 = Anim Dynamics - Bone: dyn_L_collarFront_1!(SimulationSpace:AnimPhysSimSpaceType Enum, ComponentPose=@n7, Alpha=1) -> (Pose) {
    angularDampingOverride=1
    angularSpringConstant=10000
    bAngularSpring=true
    bOverrideAngularDamping=true
    bUseGravityOverride=true
    gravityScale=0.35
    simulationSpace=RootRelative
    alphaBoolBlend {blendOption=Linear}
    boundBone {boneName=dyn_L_collarFront_1}
    componentAppliedLinearAccClamp=<5000 5000 5000>
    componentLinearAccScale=<1 1 1>
    componentLinearVelScale=<1 1 1>
    gravityOverride {z=-980}
    physicsBodyDefinitions [{
      boundBone {boneName=dyn_L_collarFront_1}
      boxExtents=<27 27 27>
      constraintSetup {angularLimitsMax=<0 0 10>, angularLimitsMin=<0 -10 -10>, angularTarget=<1 0 0>}
      localJointOffset=<3 0 0>
     }]
   }
   n9 = Anim Dynamics - Bone: dyn_L_collarSide_1!(SimulationSpace:AnimPhysSimSpaceType Enum, ComponentPose=@n8, Alpha=1) -> (Pose) {
    angularDampingOverride=1
    angularSpringConstant=10000
    bAngularSpring=true
    bOverrideAngularDamping=true
    bUseGravityOverride=true
    gravityScale=0.35
    simulationSpace=RootRelative
    alphaBoolBlend {blendOption=Linear}
    boundBone {boneName=dyn_L_collarSide_1}
    componentAppliedLinearAccClamp=<5000 5000 5000>
    componentLinearAccScale=<1 1 1>
    componentLinearVelScale=<1 1 1>
    gravityOverride {z=-980}
    physicsBodyDefinitions [{
      boundBone {boneName=dyn_L_collarSide_1}
      boxExtents=<27 27 27>
      constraintSetup {angularLimitsMax=<0 0 10>, angularLimitsMin=<0 -10 -10>, angularTarget=<1 0 0>}
      localJointOffset=<3 0 0>
     }]
   }
   n10 = Anim Dynamics - Bone: dyn_L_collarBack_1!(SimulationSpace:AnimPhysSimSpaceType Enum, ComponentPose=@n9, Alpha=1) -> (Pose) {
    angularDampingOverride=1
    angularSpringConstant=10000
    bAngularSpring=true
    bOverrideAngularDamping=true
    bUseGravityOverride=true
    gravityScale=0.35
    simulationSpace=RootRelative
    alphaBoolBlend {blendOption=Linear}
    boundBone {boneName=dyn_L_collarBack_1}
    componentAppliedLinearAccClamp=<5000 5000 5000>
    componentLinearAccScale=<1 1 1>
    componentLinearVelScale=<1 1 1>
    gravityOverride {z=-980}
    physicsBodyDefinitions [{
      boundBone {boneName=dyn_L_collarBack_1}
      boxExtents=<27 27 27>
      constraintSetup {angularLimitsMax=<0 0 10>, angularLimitsMin=<0 -3 -10>, angularTarget=<1 0 0>}
      localJointOffset=<3 0 0>
     }]
   }
   n11 = Anim Dynamics - Bone: dyn_C_collarBack_1!(SimulationSpace:AnimPhysSimSpaceType Enum, ComponentPose=@n10, Alpha=1) -> (Pose) {
    angularDampingOverride=1
    angularSpringConstant=10000
    bAngularSpring=true
    bOverrideAngularDamping=true
    bUseGravityOverride=true
    gravityScale=0.35
    simulationSpace=RootRelative
    alphaBoolBlend {blendOption=Linear}
    boundBone {boneName=dyn_C_collarBack_1}
    componentAppliedLinearAccClamp=<5000 5000 5000>
    componentLinearAccScale=<1 1 1>
    componentLinearVelScale=<1 1 1>
    gravityOverride {z=-980}
    physicsBodyDefinitions [{
      boundBone {boneName=dyn_C_collarBack_1}
      boxExtents=<27 27 27>
      constraintSetup {angularLimitsMax=<0 0 10>, angularLimitsMin=<0 -3 -10>, angularTarget=<1 0 0>}
      localJointOffset=<3 0 0>
     }]
   }
   n12 = Anim Dynamics - Bone: dyn_R_collarFront_1!(SimulationSpace:AnimPhysSimSpaceType Enum, ComponentPose=@n11, Alpha=1) -> (Pose) {
    angularDampingOverride=1
    angularSpringConstant=10000
    bAngularSpring=true
    bOverrideAngularDamping=true
    bUseGravityOverride=true
    gravityScale=0.35
    simulationSpace=RootRelative
    alphaBoolBlend {blendOption=Linear}
    boundBone {boneName=dyn_R_collarFront_1}
    componentAppliedLinearAccClamp=<5000 5000 5000>
    componentLinearAccScale=<1 1 1>
    componentLinearVelScale=<1 1 1>
    gravityOverride {z=-980}
    physicsBodyDefinitions [{
      boundBone {boneName=dyn_R_collarFront_1}
      boxExtents=<27 27 27>
      constraintSetup {angularLimitsMax=<0 0 10>, angularLimitsMin=<0 -10 -10>, angularTarget=<1 0 0>}
      localJointOffset=<-3 0 0>
     }]
   }
   n13 = Anim Dynamics - Bone: dyn_R_collarSide_1!(SimulationSpace:AnimPhysSimSpaceType Enum, ComponentPose=@n12, Alpha=1) -> (Pose) {
    angularDampingOverride=1
    angularSpringConstant=10000
    bAngularSpring=true
    bOverrideAngularDamping=true
    bUseGravityOverride=true
    gravityScale=0.35
    simulationSpace=RootRelative
    alphaBoolBlend {blendOption=Linear}
    boundBone {boneName=dyn_R_collarSide_1}
    componentAppliedLinearAccClamp=<5000 5000 5000>
    componentLinearAccScale=<1 1 1>
    componentLinearVelScale=<1 1 1>
    gravityOverride {z=-980}
    physicsBodyDefinitions [{
      boundBone {boneName=dyn_R_collarSide_1}
      boxExtents=<27 27 27>
      constraintSetup {angularLimitsMax=<0 0 10>, angularLimitsMin=<0 -10 -10>, angularTarget=<1 0 0>}
      localJointOffset=<-3 0 0>
     }]
   }
   n14 = Anim Dynamics - Bone: dyn_R_collarBack_1!(SimulationSpace:AnimPhysSimSpaceType Enum, ComponentPose=@n13, Alpha=1) -> (Pose) {
    angularDampingOverride=1
    angularSpringConstant=10000
    bAngularSpring=true
    bOverrideAngularDamping=true
    bUseGravityOverride=true
    gravityScale=0.35
    simulationSpace=RootRelative
    alphaBoolBlend {blendOption=Linear}
    boundBone {boneName=dyn_R_collarBack_1}
    componentAppliedLinearAccClamp=<5000 5000 5000>
    componentLinearAccScale=<1 1 1>
    componentLinearVelScale=<1 1 1>
    gravityOverride {z=-980}
    physicsBodyDefinitions [{
      boundBone {boneName=dyn_R_collarBack_1}
      boxExtents=<27 27 27>
      constraintSetup {angularLimitsMax=<0 0 10>, angularLimitsMin=<0 -3 -10>, angularTarget=<1 0 0>}
      localJointOffset=<-3 0 0>
     }]
   }
   n15 = Anim Dynamics - Bone: dyn_belt_back_1!(SimulationSpace:AnimPhysSimSpaceType Enum, ComponentPose=@n14) -> (Pose) {
    angularDampingOverride=0.95
    angularSpringConstant=2500
    bAngularSpring=true
    bOverrideAngularDamping=true
    bUseGravityOverride=true
    gravityScale=0.35
    simulationSpace=RootRelative
    alphaBoolBlend {blendOption=Linear}
    boundBone {boneName=dyn_belt_back_1}
    componentAppliedLinearAccClamp=<5000 5000 5000>
    componentLinearAccScale=<1 1 1>
    componentLinearVelScale=<1 1 1>
    gravityOverride {z=-2000}
    physicsBodyDefinitions [{
      boundBone {boneName=dyn_belt_back_1}
      boxExtents=<17 17 17>
      constraintSetup {angularLimitsMax=<0 0 0>, angularLimitsMin=<0 -70 -5>, angularTarget=<1 0 0>}
      localJointOffset=<3 0 0>
     }]
   }
   n16 = Anim Dynamics - Bone: dyn_belt_front_1!(SimulationSpace:AnimPhysSimSpaceType Enum, ComponentPose=@n15) -> (Pose) {
    angularDampingOverride=0.99
    angularSpringConstant=10000
    bAngularSpring=true
    bOverrideAngularDamping=true
    bUseGravityOverride=true
    gravityScale=0.35
    simulationSpace=RootRelative
    alphaBoolBlend {blendOption=Linear}
    boundBone {boneName=dyn_belt_front_1}
    componentAppliedLinearAccClamp=<5000 5000 5000>
    componentLinearAccScale=<1 1 1>
    componentLinearVelScale=<1 1 1>
    gravityOverride {z=-2000}
    physicsBodyDefinitions [{
      boundBone {boneName=dyn_belt_front_1}
      boxExtents=<23 23 23>
      constraintSetup {angularLimitsMax=<0 0 30>, angularLimitsMin=<0 -60 0>, angularTarget=<1 0 0>}
      localJointOffset=<2 0 0>
     }]
   }
   n17 = Anim Dynamics - Bone: dyn_belt_mid_1!(SimulationSpace:AnimPhysSimSpaceType Enum, ComponentPose=@n16) -> (Pose) {
    angularDampingOverride=0.99
    angularSpringConstant=3000
    bAngularSpring=true
    bOverrideAngularDamping=true
    bUseGravityOverride=true
    gravityScale=0.35
    simulationSpace=RootRelative
    alphaBoolBlend {blendOption=Linear}
    boundBone {boneName=dyn_belt_mid_1}
    componentAppliedLinearAccClamp=<5000 5000 5000>
    componentLinearAccScale=<1 1 1>
    componentLinearVelScale=<1 1 1>
    gravityOverride {z=-2000}
    physicsBodyDefinitions [{
      boundBone {boneName=dyn_belt_mid_1}
      boxExtents=<18 18 18>
      constraintSetup {angularLimitsMax=<0 0 0>, angularLimitsMin=<0 -60 0>, angularTarget=<1 0 0>}
      localJointOffset=<2 0 0>
     }]
   }
   n18 = "Constraint - dyn_zipper (PPPPPPPP)"!(
    ComponentPose=@n17
    Alpha=1
    ConstraintWeights_0=0.0002
    ConstraintWeights_1:float=0.0361
    ConstraintWeights_2:float=0.066
    ConstraintWeights_3:float=0.205
    ConstraintWeights_4:float=0.6846
    ConstraintWeights_5:float=0.0057
    ConstraintWeights_6:float=0.0007
    ConstraintWeights_7:float=0.0017
   ) -> (Pose) {
    alphaBoolBlend {blendOption=Linear}
    boneToModify {boneName=dyn_zipper}
    constraintSetup [
     {targetBone {boneName=pelvis}}
     {targetBone {boneName=spine_01}}
     {targetBone {boneName=spine_02}}
     {targetBone {boneName=spine_03}}
     {targetBone {boneName=spine_04}}
     {targetBone {boneName=spine_05}}
     {targetBone {boneName=clavicle_r}}
     {targetBone {boneName=upperarm_twist_01_r}}
    ]
    constraintWeights [1,1,1,1,1,1,1,1]
   }
   n19 = Anim Dynamics - Bone: dyn_zipper(ComponentPose=@n18, Alpha=1) {
    angularDampingOverride=1
    angularSpringConstant=6000
    bAngularSpring=true
    bOverrideAngularDamping=true
    bUseGravityOverride=true
    gravityScale=0.35
    simulationSpace=RootRelative
    alphaBoolBlend {blendOption=Linear}
    boundBone {boneName=dyn_zipper}
    componentLinearAccScale=<1 1 1>
    componentLinearVelScale=<1 1 1>
    gravityOverride {z=-980}
    physicsBodyDefinitions [{
      boundBone {boneName=dyn_zipper}
      boxExtents=<13 13 13>
      constraintSetup {angularLimitsMax=<0 10 80>, angularLimitsMin=<0 -10 0>, angularTarget=<1 0 0>}
      localJointOffset=<3 0 0>
     }]
   }
   Output Pose(Result=Component To Local(ComponentPose=@n19)) {layerGroup=None}
  ]}
 EventGraph {nodes [
   Event BlueprintUpdateAnimation()
   > n1 = Sequence()
   > n2 = Branch(Condition=OR Boolean(A=Get bIsDBNOCarried(), B=Get bIsInFrontEnd()))
   > Set simSpaceChange(simSpaceChange:AnimPhysSimSpaceType Enum=World) -> (Output_Get:AnimPhysSimSpaceType Enum)
   Set simSpaceChange(execute=@n2.else)
   Set frontendAlpha(execute=@n1.then_1, frontendAlpha:double=SelectFloat(A=0.5, B=1, bPickA=Get bIsInFrontEnd())) -> (Output_Get:double)
   > Branch(Condition=Get bIsInFrontEnd())
   > Set Custom Idle is Playing Float(Custom Idle is Playing Float:float=GetCurveValue(self=GetOwnerFrontEndAnimInstance(), CurveName=Frontend_Idle_Bypass)) -> (Output_Get:float)
  ]}
]
variables [
 Custom Idle is Playing Float {type=number}
 frontendAlpha {type=number}
 simSpaceChange {type=AnimPhysSimSpaceType}
]
