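# Animation Blueprint for the M_MED_FreeDrive face accessory, adding secondary jiggle to its burger-like parts: eyes, tongue, melty cheese and top bun.
# It copies the attached parent mesh's pose, then chains per-bone AnimDynamics: eyes slide on one axis, while other parts use limited angular springs.
# Additive Modify Bone nodes then offset the tongue's rotation.
# The event graph sets dynamics-disabled and simulation-space variables from game state, but the anim graph shows no reads of either.
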AnimBlueprint /BRCosmetics/Characters/Player/Male/Medium/Bodies/M_MED_FreeDrive/Meshes/Parts/M_MED_FreeDrive_FaceAcc_AnimBP
target_skeleton=&/M_MED_FreeDrive_FaceAcc_Skeleton
graphs [
 AnimGraph {nodes [
   n1 = Anim Dynamics - Bone: dyn_L_eye(ComponentPose=Local To Component(LocalPose=Copy Pose From Mesh() {bUseAttachedParent=true}), Alpha=1, ExternalForce="0,0,0") {
    angularDampingOverride=1
    bOverrideLinearDamping=true
    bUseGravityOverride=true
    gravityScale=0.35
    linearDampingOverride=1
    simulationSpace=RootRelative
    alphaBoolBlend {blendOption=Linear}
    boundBone {boneName=dyn_L_eye}
    componentAppliedLinearAccClamp=<1000 1000 1000>
    componentLinearAccScale=<1 1 1>
    componentLinearVelScale=<1 1 1>
    gravityOverride {z=-980}
    physicsBodyDefinitions [{
      boundBone {boneName=dyn_L_eye}
      boxExtents=<12 12 12>
      constraintSetup {angularLimitsMax=<0 0 0>, angularLimitsMin=<0 0 0>, angularTarget=<1 0 0>, linearAxesMax=<0 0 1>}
      localJointOffset=<0 0 0>
     }]
   }
   n2 = Anim Dynamics - Bone: dyn_R_eye(ComponentPose=@n1, Alpha=1, ExternalForce="0,0,0") {
    angularDampingOverride=1
    bOverrideLinearDamping=true
    bUseGravityOverride=true
    gravityScale=0.35
    linearDampingOverride=1
    simulationSpace=RootRelative
    alphaBoolBlend {blendOption=Linear}
    boundBone {boneName=dyn_R_eye}
    componentAppliedLinearAccClamp=<1000 1000 1000>
    componentLinearAccScale=<1 1 1>
    componentLinearVelScale=<1 1 1>
    gravityOverride {z=-980}
    physicsBodyDefinitions [{
      boundBone {boneName=dyn_R_eye}
      boxExtents=<12 12 12>
      constraintSetup {angularLimitsMax=<0 0 0>, angularLimitsMin=<0 0 0>, angularTarget=<1 0 0>, linearAxesMax=<0 0 2>}
      localJointOffset=<0 0 0>
     }]
   }
   n3 = Anim Dynamics - Bone: dyn_tongue_1!(SimulationSpace:AnimPhysSimSpaceType Enum, ComponentPose=@n2, Alpha=1) -> (Pose) {
    angularDampingOverride=1
    angularSpringConstant=20000
    bAngularSpring=true
    bOverrideAngularDamping=true
    bUseGravityOverride=true
    gravityScale=0.35
    simulationSpace=RootRelative
    alphaBoolBlend {blendOption=Linear}
    boundBone {boneName=dyn_tongue_1}
    componentAppliedLinearAccClamp=<1000 1000 1000>
    componentLinearAccScale=<1 1 1>
    componentLinearVelScale=<1 1 1>
    gravityOverride {z=-980}
    physicsBodyDefinitions [{
      boundBone {boneName=dyn_tongue_1}
      boxExtents=<25 25 25>
      constraintSetup {angularLimitsMax=<15 15 15>, angularLimitsMin=<-15 -15 -15>, angularTarget=<1 0 0>}
      localJointOffset=<4 0 0>
     }]
   }
   n4 = Anim Dynamics - Bone: dyn_tongue_2!(SimulationSpace:AnimPhysSimSpaceType Enum, ComponentPose=@n3, Alpha=1) -> (Pose) {
    angularDampingOverride=1
    angularSpringConstant=5000
    bAngularSpring=true
    bOverrideAngularDamping=true
    bUseGravityOverride=true
    gravityScale=0.35
    simulationSpace=RootRelative
    alphaBoolBlend {blendOption=Linear}
    boundBone {boneName=dyn_tongue_2}
    componentAppliedLinearAccClamp=<1000 1000 1000>
    componentLinearAccScale=<1 1 1>
    componentLinearVelScale=<1 1 1>
    gravityOverride {z=-980}
    physicsBodyDefinitions [{
      boundBone {boneName=dyn_tongue_2}
      boxExtents=<25 25 25>
      constraintSetup {angularLimitsMax=<15 15 15>, angularLimitsMin=<-15 -15 -60>, angularTarget=<1 0 0>}
      localJointOffset=<4 0 0>
     }]
   }
   n5 = Anim Dynamics - Bone: dyn_tongue_3!(SimulationSpace:AnimPhysSimSpaceType Enum, ComponentPose=@n4, Alpha=1) -> (Pose) {
    angularDampingOverride=1
    angularSpringConstant=10000
    bAngularSpring=true
    bOverrideAngularDamping=true
    bUseGravityOverride=true
    gravityScale=0.35
    simulationSpace=RootRelative
    alphaBoolBlend {blendOption=Linear}
    boundBone {boneName=dyn_tongue_3}
    componentAppliedLinearAccClamp=<1000 1000 1000>
    componentLinearAccScale=<1 1 1>
    componentLinearVelScale=<1 1 1>
    gravityOverride {z=-980}
    physicsBodyDefinitions [{
      boundBone {boneName=dyn_tongue_3}
      boxExtents=<25 25 25>
      constraintSetup {angularLimitsMax=<15 15 15>, angularLimitsMin=<-15 -15 -60>, angularTarget=<1 0 0>}
      localJointOffset=<4 0 0>
     }]
   }
   n6 = Anim Dynamics - Bone: dyn_cheeseCorner!(SimulationSpace:AnimPhysSimSpaceType Enum, ComponentPose=@n5, Alpha=1) -> (Pose) {
    angularDampingOverride=1
    angularSpringConstant=10000
    bAngularSpring=true
    bOverrideAngularDamping=true
    bUseGravityOverride=true
    gravityScale=0.35
    simulationSpace=RootRelative
    alphaBoolBlend {blendOption=Linear}
    boundBone {boneName=dyn_cheeseCorner}
    componentAppliedLinearAccClamp=<1000 1000 1000>
    componentLinearAccScale=<1 1 1>
    componentLinearVelScale=<1 1 1>
    gravityOverride {z=-980}
    physicsBodyDefinitions [{
      boundBone {boneName=dyn_cheeseCorner}
      boxExtents=<20 30 30>
      constraintSetup {angularLimitsMax=<15 5 15>, angularLimitsMin=<-15 -5 -5>, angularTarget=<1 0 0>}
      localJointOffset=<4 0 0>
     }]
   }
   n7 = Anim Dynamics - Bone: dyn_meltyCheese_1!(SimulationSpace:AnimPhysSimSpaceType Enum, ComponentPose=@n6, Alpha=1) -> (Pose) {
    angularDampingOverride=1
    angularSpringConstant=20000
    bAngularSpring=true
    bOverrideAngularDamping=true
    bUseGravityOverride=true
    gravityScale=0.35
    simulationSpace=RootRelative
    alphaBoolBlend {blendOption=Linear}
    boundBone {boneName=dyn_meltyCheese_1}
    componentAppliedLinearAccClamp=<1000 1000 1000>
    componentLinearAccScale=<1 1 1>
    componentLinearVelScale=<1 1 1>
    gravityOverride {z=-980}
    physicsBodyDefinitions [{
      boundBone {boneName=dyn_meltyCheese_1}
      boxExtents=<20 30 30>
      constraintSetup {angularLimitsMax=<15 15 15>, angularLimitsMin=<-15 -5 -15>, angularTarget=<1 0 0>}
      localJointOffset=<4 0 0>
     }]
   }
   n8 = Anim Dynamics - Bone: dyn_meltyCheese_2!(SimulationSpace:AnimPhysSimSpaceType Enum, ComponentPose=@n7, Alpha=1) -> (Pose) {
    angularDampingOverride=1
    angularSpringConstant=20000
    bAngularSpring=true
    bOverrideAngularDamping=true
    bUseGravityOverride=true
    gravityScale=0.35
    simulationSpace=RootRelative
    alphaBoolBlend {blendOption=Linear}
    boundBone {boneName=dyn_meltyCheese_2}
    componentAppliedLinearAccClamp=<1000 1000 1000>
    componentLinearAccScale=<1 1 1>
    componentLinearVelScale=<1 1 1>
    gravityOverride {z=-980}
    physicsBodyDefinitions [{
      boundBone {boneName=dyn_meltyCheese_2}
      boxExtents=<20 30 30>
      constraintSetup {angularLimitsMax=<15 15 15>, angularLimitsMin=<-15 -15 -15>, angularTarget=<1 0 0>}
      localJointOffset=<4 0 0>
     }]
   }
   n9 = Anim Dynamics - Bone: dyn_littleMeltyCheese!(SimulationSpace:AnimPhysSimSpaceType Enum, ComponentPose=@n8, Alpha=1) -> (Pose) {
    angularDampingOverride=1
    angularSpringConstant=20000
    bAngularSpring=true
    bOverrideAngularDamping=true
    bUseGravityOverride=true
    gravityScale=0.35
    simulationSpace=RootRelative
    alphaBoolBlend {blendOption=Linear}
    boundBone {boneName=dyn_littleMeltyCheese}
    componentAppliedLinearAccClamp=<1000 1000 1000>
    componentLinearAccScale=<1 1 1>
    componentLinearVelScale=<1 1 1>
    gravityOverride {z=-980}
    physicsBodyDefinitions [{
      boundBone {boneName=dyn_littleMeltyCheese}
      boxExtents=<20 30 30>
      constraintSetup {angularLimitsMax=<10 15 15>, angularLimitsMin=<-10 0 -15>, angularTarget=<1 0 0>}
      localJointOffset=<4 0 0>
     }]
   }
   n10 = Anim Dynamics - Bone: dyn_meltyCheese_L_back!(SimulationSpace:AnimPhysSimSpaceType Enum, ComponentPose=@n9, Alpha=1) -> (Pose) {
    angularDampingOverride=0.999
    angularSpringConstant=15000
    bAngularSpring=true
    bOverrideAngularDamping=true
    bUseGravityOverride=true
    gravityScale=0.35
    simulationSpace=RootRelative
    alphaBoolBlend {blendOption=Linear}
    boundBone {boneName=dyn_meltyCheese_L_back}
    componentAppliedLinearAccClamp=<1000 1000 1000>
    componentLinearAccScale=<1 1 1>
    componentLinearVelScale=<1 1 1>
    gravityOverride {z=-980}
    physicsBodyDefinitions [{
      boundBone {boneName=dyn_meltyCheese_L_back}
      boxExtents=<20 30 30>
      constraintSetup {angularLimitsMax=<15 5 15>, angularLimitsMin=<-15 -5 -15>, angularTarget=<1 0 0>}
      localJointOffset=<3 0 0>
     }]
   }
   n11 = Anim Dynamics - Bone: dyn_meltyCheese_R_back_1!(SimulationSpace:AnimPhysSimSpaceType Enum, ComponentPose=@n10, Alpha=1) -> (Pose) {
    angularDampingOverride=0.999
    angularSpringConstant=15000
    bAngularSpring=true
    bOverrideAngularDamping=true
    bUseGravityOverride=true
    gravityScale=0.35
    simulationSpace=RootRelative
    alphaBoolBlend {blendOption=Linear}
    boundBone {boneName=dyn_meltyCheese_R_back_1}
    componentAppliedLinearAccClamp=<1000 1000 1000>
    componentLinearAccScale=<1 1 1>
    componentLinearVelScale=<1 1 1>
    gravityOverride {z=-980}
    physicsBodyDefinitions [{
      boundBone {boneName=dyn_meltyCheese_R_back_1}
      boxExtents=<20 20 20>
      constraintSetup {angularLimitsMax=<15 15 25>, angularLimitsMin=<-15 -5 -25>, angularTarget=<1 0 0>}
      localJointOffset=<3 0 0>
     }]
   }
   n12 = Anim Dynamics - Bone: dyn_meltyCheese_R_back_2!(SimulationSpace:AnimPhysSimSpaceType Enum, ComponentPose=@n11, Alpha=1) -> (Pose) {
    angularDampingOverride=0.999
    angularSpringConstant=15000
    bAngularSpring=true
    bOverrideAngularDamping=true
    bUseGravityOverride=true
    gravityScale=0.35
    simulationSpace=RootRelative
    alphaBoolBlend {blendOption=Linear}
    boundBone {boneName=dyn_meltyCheese_R_back_2}
    componentAppliedLinearAccClamp=<1000 1000 1000>
    componentLinearAccScale=<1 1 1>
    componentLinearVelScale=<1 1 1>
    gravityOverride {z=-980}
    physicsBodyDefinitions [{
      boundBone {boneName=dyn_meltyCheese_R_back_2}
      boxExtents=<27 27 27>
      constraintSetup {angularLimitsMax=<15 15 25>, angularLimitsMin=<-15 -5 -25>, angularTarget=<1 0 0>}
      localJointOffset=<3 0 0>
     }]
   }
   n13 = Anim Dynamics - Bone: dyn_cheeseDrop_back!(SimulationSpace:AnimPhysSimSpaceType Enum, ComponentPose=@n12, Alpha=1) -> (Pose) {
    angularDampingOverride=0.999
    angularSpringConstant=15000
    bAngularSpring=true
    bOverrideAngularDamping=true
    bUseGravityOverride=true
    gravityScale=0.35
    simulationSpace=RootRelative
    alphaBoolBlend {blendOption=Linear}
    boundBone {boneName=dyn_cheeseDrop_back}
    componentAppliedLinearAccClamp=<1000 1000 1000>
    componentLinearAccScale=<1 1 1>
    componentLinearVelScale=<1 1 1>
    gravityOverride {z=-980}
    physicsBodyDefinitions [{
      boundBone {boneName=dyn_cheeseDrop_back}
      boxExtents=<30 30 30>
      constraintSetup {angularLimitsMax=<10 15 15>, angularLimitsMin=<-10 0 -15>, angularTarget=<1 0 0>}
      localJointOffset=<3 0 0>
     }]
   }
   n14 = Anim Dynamics - Bone: dyn_topBun(ComponentPose=@n13, Alpha=1, ExternalForce="0,0,0") {
    angularDampingOverride=1
    angularSpringConstant=5000
    bAngularSpring=true
    bOverrideAngularDamping=true
    bUseGravityOverride=true
    gravityScale=0.35
    simulationSpace=RootRelative
    alphaBoolBlend {blendOption=Linear}
    boundBone {boneName=dyn_topBun}
    componentAppliedLinearAccClamp=<1000 1000 1000>
    componentLinearAccScale=<1 1 1>
    componentLinearVelScale=<1 1 1>
    gravityOverride {z=-600}
    physicsBodyDefinitions [{
      boundBone {boneName=dyn_topBun}
      boxExtents=<23 23 23>
      constraintSetup {angularLimitsMax=<5 0 0>, angularLimitsMin=<-5 0 -10>, angularTarget=<1 0 0>}
      localJointOffset=<3 0 0>
     }]
   }
   n15 = "Transform (Modify) Bone - Bone: dyn_tongue_1"!(ComponentPose=@n14, Alpha=1, Rotation="-0.000,-20.000,-0.000") -> (Pose) {
    rotationMode=BMM_Additive
    rotationSpace=BCS_BoneSpace
    alphaBoolBlend {blendOption=Linear}
    boneToModify {boneName=dyn_tongue_1}
    rotation=<rot 0 -20 0>
   }
   n16 = "Transform (Modify) Bone - Bone: dyn_tongue_2"!(ComponentPose=@n15, Alpha=1, Rotation="-0.000,-45.000,-0.000") -> (Pose) {
    rotationMode=BMM_Additive
    rotationSpace=BCS_BoneSpace
    alphaBoolBlend {blendOption=Linear}
    boneToModify {boneName=dyn_tongue_2}
    rotation=<rot 0 -45 0>
   }
   n17 = "Transform (Modify) Bone - Bone: dyn_tongue_3"!(ComponentPose=@n16, Alpha=1, Rotation="-0.000,-15.000,-0.000") -> (Pose) {
    rotationMode=BMM_Additive
    rotationSpace=BCS_BoneSpace
    alphaBoolBlend {blendOption=Linear}
    boneToModify {boneName=dyn_tongue_3}
    rotation=<rot 0 -15 0>
   }
   Output Pose(Result=Component To Local(ComponentPose=@n17)) {layerGroup=None}
  ]}
 EventGraph {nodes [
   Event BlueprintUpdateAnimation()
   > Set animDynamicsDisabled(animDynamicsDisabled:bool=NOT Boolean(A=Get bAnimDynamicsEnabled() -> (bAnimDynamicsEnabled:bool))) -> (Output_Get:bool)
   > n1 = Branch(Condition=OR Boolean(A=Get bIsDBNO(), B=Get bIsInFrontEnd()))
   > Set simSpaceChange(simSpaceChange:AnimPhysSimSpaceType Enum=World) -> (Output_Get:AnimPhysSimSpaceType Enum)
   Set simSpaceChange(execute=@n1.else)
  ]}
]
variables [
 animDynamicsDisabled {type=boolean}
 simSpaceChange {type=AnimPhysSimSpaceType}
]
